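# Conversation database for the BrightDisk NPC, opening player dialogue at the Conversation.Entry.BrightDisk entry point.
# A single speech node picks a greeting by priority and weight from first-meeting, per-match conversation count and outfit-specific lines, the last keyed to worn character cosmetics.
# That node fans out to link nodes that jump by tag to remote conversations for NPC quests, basic and random services, further dialogue, and seasonal story and weekly quests.
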ConversationDatabase /NPCLibrary/NPCs/BrightDisk/NPCLibrary_Conversation_BrightDisk
entry_points [{tag=Conversation.Entry.BrightDisk, destinations [#1]}]
nodes [
 ConversationEntryPointNode_7 {
  id=#1
  class=ConversationEntryPointNode
  outputs [#2]
  properties {entryTag {tagName=Conversation.Entry.BrightDisk}}
 }
 Speech_C_4 {
  id=#2
  class=Speech_C
  outputs [
   #3
   #4
   #5
   #6
   #7
   #8
   #9
   #10
   #11
   #12
   #13
   #14
  ]
  properties {
   nodeName=""
   generalConfig {
    defaultMessage="Here we go again, huh?"
    contextualMessages [
     {
      message="Here we go again, huh?"
      contextRequirements [{requirement {class=FortControllerRequirement_IsFirstEverConversationWithNPC}}]
      priority {curve {rowName=Tandem.Speech.Priority.FirstEverInteraction}}
     }
     {
      message=That Dark Voyager... What does he want?
      contextRequirements [{requirement {class=FortControllerRequirement_CheckNPCMatchConversationCount, properties {nPCUniqueIdTag {tagName=AISpawnerData.Type.Tandem.BrightDisk}}}}]
      priority {curve {curveTable=null, rowName=None}}
     }
     {
      message=I wish I could see what's coming. Anything...
      contextRequirements [{requirement {class=FortControllerRequirement_CheckNPCMatchConversationCount, properties {nPCUniqueIdTag {tagName=AISpawnerData.Type.Tandem.BrightDisk}}}}]
      priority {curve {curveTable=null, rowName=None}}
     }
     {
      message=This place is so beautiful. We have to protect it.
      contextRequirements [{requirement {class=FortControllerRequirement_CheckNPCMatchConversationCount, properties {nPCUniqueIdTag {tagName=AISpawnerData.Type.Tandem.BrightDisk}}}}]
      priority {curve {curveTable=null, rowName=None}}
      weight {value=2}
     }
     {
      message=Jones and I have come a long way. I can't let him down.
      contextRequirements [{requirement {class=FortControllerRequirement_CheckNPCMatchConversationCount, properties {nPCUniqueIdTag {tagName=AISpawnerData.Type.Tandem.BrightDisk}}}}]
      priority {curve {curveTable=null, rowName=None}}
     }
     {
      message="My visions, they're... I shouldn't talk about this."
      contextRequirements [{requirement {class=FortControllerRequirement_CheckNPCMatchConversationCount, properties {nPCUniqueIdTag {tagName=AISpawnerData.Type.Tandem.BrightDisk}}}}]
      priority {curve {curveTable=null, rowName=None}}
     }
     {
      message=Dark Voyager is collecting the Zero Point Shards. How do we stop him?
      contextRequirements [{requirement {
         class=FortControllerRequirement_HasCID
         properties {allowedCIds [&Character_UndergroundRebel_Fashion]}
        }}]
      priority {curve {rowName=Tandem.Speech.Priority.OutfitReaction}}
     }
     {
      message=What is it that you want?
      contextRequirements [{requirement {class=FortControllerRequirement_HasCID, properties {allowedCIds [/BRCosmetics/Athena/Items/Cosmetics/Characters/Character_SpacePlunge]}}}]
      priority {curve {rowName=Tandem.Speech.Priority.OutfitReaction}}
     }
     {
      message="Hey, buddy! Looking fresh!"
      contextRequirements [{requirement {
         class=FortControllerRequirement_HasCID
         properties {allowedCIds [
           &Character_BananaPhilosopher
           &Character_BananaAdventure
           /BRCosmetics/Athena/Items/Cosmetics/Characters/Character_BananaCake
           &CID_349_Athena_Commando_M_Banana
           &CID_701_Athena_Commando_M_BananaAgent
           &CID_814_Athena_Commando_M_BananaSummer
           /BRCosmetics/Athena/Items/Cosmetics/Characters/CID_A_045_Athena_Commando_M_BananaLeader
           &CID_A_323_Athena_Commando_M_BananaWinter
           /BRCosmetics/Athena/Items/Cosmetics/Characters/CID_TBD_Athena_Commando_M_Banana_CINE
          ]}
        }}]
      priority {curve {rowName=Tandem.Speech.Priority.OutfitReaction}}
     }
     {
      message=Please stay out of our business. It's more dangerous than you know.
      contextRequirements [{requirement {class=FortControllerRequirement_HasCID, properties {allowedCIds [/BRCosmetics/Athena/Items/Cosmetics/Characters/Character_MarineCarve]}}}]
      priority {curve {rowName=Tandem.Speech.Priority.OutfitReaction}}
     }
    ]
   }
  }
 }
 ConversationLinkNode_11 {
  id=#14
  class=ConversationLinkNode
  properties {remoteEntryTag {tagName=Conversation.Entry.Season39.Weekly.W15.CQ02}}
 }
 ConversationLinkNode_5 {
  id=#13
  class=ConversationLinkNode
  properties {remoteEntryTag {tagName=Conversation.Entry.BrightDisk.Dialogue}}
 }
 ConversationLinkNode_10 {
  id=#12
  class=ConversationLinkNode
  properties {remoteEntryTag {tagName=Conversation.Entry.Season39.Weekly.W09.Q01.BrightDisk}}
 }
 ConversationLinkNode_9 {
  id=#11
  class=ConversationLinkNode
  properties {remoteEntryTag {tagName=Conversation.Entry.Season39.Weekly.W05.Q01.BrightDisk}}
 }
 ConversationLinkNode_8 {
  id=#10
  class=ConversationLinkNode
  properties {remoteEntryTag {tagName=Conversation.Entry.Season39.WF.PerfectGift}}
 }
 ConversationLinkNode_7 {id=#9, class=ConversationLinkNode, properties {remoteEntryTag {tagName=Conversation.Entry.RandomServices}}}
 ConversationLinkNode_6 {
  id=#8
  class=ConversationLinkNode
  properties {remoteEntryTag {tagName=Conversation.Entry.Season39.BrightDisk.BRQuest.01}}
 }
 ConversationLinkNode_4 {
  id=#7
  class=ConversationLinkNode
  properties {remoteEntryTag {tagName=Conversation.Entry.Season31.Story.P06.Q06}}
 }
 ConversationLinkNode_3 {
  id=#6
  class=ConversationLinkNode
  properties {remoteEntryTag {tagName=Conversation.Entry.Season31.Story.P03.Q07b}}
 }
 ConversationLinkNode_2 {
  id=#5
  class=ConversationLinkNode
  properties {remoteEntryTag {tagName=Conversation.Entry.Season31.Story.P01.Q03}}
 }
 ConversationLinkNode_0 {id=#4, class=ConversationLinkNode, properties {remoteEntryTag {tagName=Conversation.Entry.BasicServices}}}
 ConversationLinkNode_1 {id=#3, class=ConversationLinkNode, properties {remoteEntryTag {tagName=Conversation.Entry.NPCLibrary.Quests}}}
]
