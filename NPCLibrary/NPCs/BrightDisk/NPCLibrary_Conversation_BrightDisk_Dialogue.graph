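# Conversation database for the BrightDisk NPC, giving players lore dialogue under the Conversation.Entry.BrightDisk.Dialogue entry tag.
# An "Ask About..." speech node acts as the hub, branching to Oninoshima, Zero Point Shards and The Dark Voyager topics, plus a Back choice.
# Each topic is a linear chain of Speech nodes advanced by "Continue..." choices, ending in a ReturnToStart node labeled BACK.
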ConversationDatabase /NPCLibrary/NPCs/BrightDisk/NPCLibrary_Conversation_BrightDisk_Dialogue
entry_points [{tag=Conversation.Entry.BrightDisk.Dialogue, destinations [#1]}]
nodes [
 ConversationEntryPointNode_0 {
  id=#1
  class=ConversationEntryPointNode
  outputs [#2]
  properties {entryTag {tagName=Conversation.Entry.BrightDisk.Dialogue}}
 }
 Speech_C_16 {
  id=#2
  class=Speech_C
  outputs [#3, #4, #5, #6]
  properties {nodeName="", softTaskIcon=&T_UI_NPC_question, generalConfig {defaultMessage=How can I help?}}
  sub_nodes [ConversationChoiceNode_3 {class=ConversationChoiceNode, properties {defaultChoiceDisplayText=Ask About...}}]
 }
 Speech_C_13 {
  id=#6
  class=Speech_C
  outputs [#7]
  properties {
   nodeName=""
   softTaskIcon=&T_UI_NPC_question
   generalConfig {defaultMessage="Whoever this Dark Voyager is: he's powerful, he's dangerous, and he's collecting the Zero Point Shards."}
  }
  sub_nodes [ConversationChoiceNode_6 {class=ConversationChoiceNode, properties {defaultChoiceDisplayText=The Dark Voyager}}]
 }
 Speech_C_4 {
  id=#7
  class=Speech_C
  outputs [#8]
  properties {
   nodeName=""
   softTaskIcon=&T_UI_NPC_question
   generalConfig {defaultMessage="We're building up our resources to defeat him, but..."}
  }
  sub_nodes [ConversationChoiceNode_2 {class=ConversationChoiceNode, properties {defaultChoiceDisplayText=Continue...}}]
 }
 Speech_C_6 {
  id=#8
  class=Speech_C
  outputs [#9]
  properties {
   nodeName=""
   softTaskIcon=&T_UI_NPC_question
   generalConfig {defaultMessage="I'd stay away from him, at least until we know more about his capaibilities."}
  }
  sub_nodes [ConversationChoiceNode_8 {class=ConversationChoiceNode, properties {defaultChoiceDisplayText=Continue...}}]
 }
 ReturnToStart_C_2 {
  id=#9
  class=ReturnToStart_C
  properties {softTaskIcon=&T_UI_NPC_Back}
  sub_nodes [ConversationChoiceNode_1 {class=ConversationChoiceNode, properties {defaultChoiceDisplayText=BACK}}]
 }
 Speech_C_8 {
  id=#5
  class=Speech_C
  outputs [#10]
  properties {
   nodeName=""
   softTaskIcon=&T_UI_NPC_question
   generalConfig {defaultMessage=The Dark Voyager has gotten his hands on a third Zero Point Shard.}
  }
  sub_nodes [ConversationChoiceNode_0 {class=ConversationChoiceNode, properties {defaultChoiceDisplayText=Zero Point Shards}}]
 }
 Speech_C_9 {
  id=#10
  class=Speech_C
  outputs [#11]
  properties {
   nodeName=""
   softTaskIcon=&T_UI_NPC_question
   generalConfig {defaultMessage="And even if we took them from him, he's so powerful."}
  }
  sub_nodes [ConversationChoiceNode_17 {class=ConversationChoiceNode, properties {defaultChoiceDisplayText=Continue...}}]
 }
 Speech_C_10 {
  id=#11
  class=Speech_C
  outputs [#12]
  properties {
   nodeName=""
   softTaskIcon=&T_UI_NPC_question
   generalConfig {defaultMessage="We need allies, we need tech, but most of all, we need time."}
  }
  sub_nodes [ConversationChoiceNode_18 {class=ConversationChoiceNode, properties {defaultChoiceDisplayText=Continue...}}]
 }
 Speech_C_12 {
  id=#12
  class=Speech_C
  outputs [#13]
  properties {nodeName="", softTaskIcon=&T_UI_NPC_question, generalConfig {defaultMessage=Time that I'm not sure we'll have.}}
  sub_nodes [ConversationChoiceNode_19 {class=ConversationChoiceNode, properties {defaultChoiceDisplayText=Continue...}}]
 }
 ReturnToStart_C_1 {
  id=#13
  class=ReturnToStart_C
  properties {softTaskIcon=&T_UI_NPC_Back}
  sub_nodes [ConversationChoiceNode_16 {class=ConversationChoiceNode, properties {defaultChoiceDisplayText=BACK}}]
 }
 Back_C_2 {
  id=#4
  class=Back_C
  sub_nodes [ConversationChoiceNode_4 {class=ConversationChoiceNode, properties {defaultChoiceDisplayText=BACK}}]
 }
 Speech_C_11 {
  id=#3
  class=Speech_C
  outputs [#14]
  properties {
   nodeName=""
   softTaskIcon=&T_UI_NPC_question
   generalConfig {defaultMessage=We left Oninoshima after we came together to defeat the Dark Presence.}
  }
  sub_nodes [ConversationChoiceNode_10 {class=ConversationChoiceNode, properties {defaultChoiceDisplayText=Oninoshima}}]
 }
 Speech_C_3 {
  id=#14
  class=Speech_C
  outputs [#15]
  properties {
   nodeName=""
   softTaskIcon=&T_UI_NPC_question
   generalConfig {defaultMessage="In the end, we left the Spirit Siblings to protect their home."}
  }
  sub_nodes [ConversationChoiceNode_12 {class=ConversationChoiceNode, properties {defaultChoiceDisplayText=Continue...}}]
 }
 Speech_C_5 {
  id=#15
  class=Speech_C
  outputs [#16]
  properties {
   nodeName=""
   softTaskIcon=&T_UI_NPC_question
   generalConfig {defaultMessage="Daigo, Jade, and Kendo. My... really good friend."}
  }
  sub_nodes [ConversationChoiceNode_13 {class=ConversationChoiceNode, properties {defaultChoiceDisplayText=Continue...}}]
 }
 Speech_C_2 {
  id=#16
  class=Speech_C
  outputs [#17]
  properties {
   nodeName=""
   softTaskIcon=&T_UI_NPC_question
   generalConfig {defaultMessage="Though I've heard rumors Daigo left, to search for somewhere he truly belongs..."}
  }
  sub_nodes [ConversationChoiceNode_14 {class=ConversationChoiceNode, properties {defaultChoiceDisplayText=Continue...}}]
 }
 ReturnToStart_C_0 {
  id=#17
  class=ReturnToStart_C
  properties {softTaskIcon=&T_UI_NPC_Back}
  sub_nodes [ConversationChoiceNode_11 {class=ConversationChoiceNode, properties {defaultChoiceDisplayText=BACK}}]
 }
]
